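# Physics asset for the Chainmail pickaxe mesh, giving its six-link dangling chain physical bodies and joint limits.
# A long capsule covers the two-handed handle, and small, mostly zero-length capsules cover each dyn_chain bone.
# Constraints link the bones in series from the handle down.
# The root joint swings wide on one axis and narrowly on the other, later links swing moderately on both, and every joint allows almost no twist.
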PhysicsAsset /BRCosmetics/Weapons/FORT_Melee/Pickaxe_Chainmail/Meshes/Chainmail_Axe_Physics
bodies {
 Melee_TwoHanded_Handle [{shapeType=Capsule, length=60.9804, radius=5.1369, center=<0 0 23.0845>, rotation=<rot 0 -90 0>}]
 dyn_chain_01 [{shapeType=Capsule, length=0, radius=1.2618, shapeName=dyn_chain_01_capsule, center=<2.6435 0.0008 0>, rotation=<rot 0 89.9794 -89.9998>}]
 dyn_chain_02 [{shapeType=Capsule, length=0, radius=1.2618, shapeName=dyn_chain_02_capsule, center=<1.5489 0.0001 0>, rotation=<rot 0 -89.9903 -89.9989>}]
 dyn_chain_03 [{shapeType=Capsule, length=0, radius=1.2618, shapeName=dyn_chain_03_capsule, center=<1.7067 0 0.0002>, rotation=<rot 0 -90.0009 -89.9903>}]
 dyn_chain_04 [{shapeType=Capsule, length=0, radius=1.2618, shapeName=dyn_chain_04_capsule, center=<1.5786 0 -0.0001>, rotation=<rot 0 90 -89.9971>}]
 dyn_chain_05 [{shapeType=Capsule, length=0, radius=1.2618, shapeName=dyn_chain_05_capsule, center=<1.5346 -0.0008 -0.0043>, rotation=<rot 0 90.0007 -90>}]
 dyn_chain_06 [{shapeType=Capsule, length=0.2583, radius=1.5352, shapeName=dyn_chain_06_capsule, center=<2.1698 0 0>, rotation=<rot 0 90 -90>}]
}
constraints [
 {bone1Name=dyn_chain_01, bone2Name=Melee_TwoHanded_Handle, swing1LimitDegrees=100, swing2LimitDegrees=5, twistLimitDegrees=1}
 {bone1Name=dyn_chain_02, bone2Name=dyn_chain_01, swing1LimitDegrees=35, swing2LimitDegrees=35, twistLimitDegrees=1}
 {bone1Name=dyn_chain_03, bone2Name=dyn_chain_02, swing1LimitDegrees=35, swing2LimitDegrees=35, twistLimitDegrees=1}
 {bone1Name=dyn_chain_04, bone2Name=dyn_chain_03, swing1LimitDegrees=35, swing2LimitDegrees=35, twistLimitDegrees=1}
 {bone1Name=dyn_chain_05, bone2Name=dyn_chain_04, swing1LimitDegrees=35, swing2LimitDegrees=35, twistLimitDegrees=1}
 {bone1Name=dyn_chain_06, bone2Name=dyn_chain_05, swing1LimitDegrees=35, swing2LimitDegrees=35, twistLimitDegrees=1}
]
