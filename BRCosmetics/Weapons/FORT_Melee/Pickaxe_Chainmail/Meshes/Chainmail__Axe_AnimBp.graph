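# Animation Blueprint for the Chainmail pickaxe that adds physics-driven secondary motion to the weapon mesh.
# A RigidBody node runs on the input pose in base-bone space from b_AR_Root, using an override physics asset and custom gravity.
# VelScale feeds both linear acceleration and velocity scales; each update lowers it while jumping and further while falling, restoring it otherwise.
# The AccScale variable is declared but unused.
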AnimBlueprint /BRCosmetics/Weapons/FORT_Melee/Pickaxe_Chainmail/Meshes/Chainmail__Axe_AnimBp
target_skeleton=&/Chainmail_Axe_Skeleton
graphs [
 AnimGraph {nodes [
   n1 = Get VelScale() -> (VelScale:Vector)
   n2 = RigidBody!(
    ComponentPose=Local To Component(LocalPose=Input Pose())
    Alpha
    ExternalForce
    ComponentLinearAccScale:Vector=@n1
    ComponentLinearVelScale:Vector=@n1
   ) -> (Pose) {
    bClampLinearTranslationLimitToRefPose=true
    bOverrideWorldGravity=true
    overridePhysicsAsset=&/Chainmail_Axe_Physics
    simulationSpace=BaseBoneSpace
    alphaBoolBlend {blendOption=Linear}
    baseBoneRef {boneName=b_AR_Root}
    componentAppliedLinearAccClamp=<5000 5000 5000>
    componentLinearAccScale=<0.45 0.45 0.45>
    componentLinearVelScale=<0.45 0.45 0.45>
    overrideWorldGravity {z=-980}
   }
   Output Pose(Result=Component To Local(ComponentPose=@n2)) {layerGroup=None}
  ]}
 EventGraph {nodes [
   Event BlueprintUpdateAnimation()
   > n1 = Branch(Condition=OR Boolean(A=Get bIsFalling(), B=Get bIsJumping()))
   > Set VelScale(VelScale:Vector="0.4,0.4,0.4") -> (Output_Get:Vector)
   > Branch(Condition=Get bIsFalling())
   > Set VelScale(VelScale="0.35,0.35,0.35")
   Set VelScale(execute=@n1.else, VelScale="0.45,0.45,0.45")
  ]}
]
variables [
 AccScale {type=Vector}
 VelScale {type=Vector}
]
